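; GA_Quest_Enter_Sideways is an event-triggered gameplay ability that reports the player entering the Sideways for Battle Pass Season 18 quests.
; On activation it calls the parent, then sends a complex custom stat event to the owning player's quest manager using its quest stat event tags and the event's target tags.
; It prints an optional debug message and ends immediately.
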(event Ability|EventActivateAbilityFromEvent
  (params (EventData :type Gameplay%20Event%20Data%20Structure%20%28by%20ref%29))
  (bind (_eventtag _instigator _target _optionalobject _optionalobject2 _contexthandle _instigatortags _targettags) (Utilities|Struct|BreakGameplayEventData EventData))
  (|CallParentFunction EventData)
  (Development|PrintString "@@@ ActivateAbilityFromEvent [GA_Quest_Enter_Sideways]: CALLED - Sending Entered Sideways Custom Stat Event!" false (Variables|Debug|GetbDebug))
  (Quest|SendComplexCustomStatEvent (Variables|GAVars|Data|GetOwningPlayerQuestManager) 0 (Variables|Default|GetQuestStatEventTags) _targettags 1)
  (Ability|EndAbility))
